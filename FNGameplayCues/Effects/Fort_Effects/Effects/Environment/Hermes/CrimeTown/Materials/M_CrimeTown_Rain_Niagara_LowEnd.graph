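# Low-end additive, unlit rain material for Niagara mesh particles in the CrimeTown environment.
# World position offset builds stretched sprites looping along a direction by time, wrapped to actor XY, with near/distance shrinking and an aspect-aware minimum screen size.
# Streak opacity combines a UV sphere mask, height fade, per-particle opacity and a building mask sampled by actor position within a box.
# A time-of-day function feeds a sprite scale control.
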Material /FNGameplayCues/Effects/Fort_Effects/Effects/Environment/Hermes/CrimeTown/Materials/M_CrimeTown_Rain_Niagara_LowEnd
graph {
 properties {
  bUseTranslucencyVertexFog=false
  bUsedWithInstancedStaticMeshes=true
  bUsedWithNiagaraMeshParticles=true
  blendMode=BLEND_Additive
  pixelDepthOffsetMode=PDOM_Legacy
  shadingModel=MSM_Unlit
  translucencyLightingMode=TLM_VolumetricDirectional
  twoSided=true
  usedShadingModels=Unlit
 }
 nodes [
  n1 = Subtract(A=WorldPosition().Z, B=ComponentMask(None=ObjectPositionWS()) {b=true})
  n2 = ScalarParameter() {defaultValue=10000, parameterName=FadeInAtHeight}
  n3 = LinearInterpolate(Alpha=Clamp(None=Divide(A=@n1, B=@n2) {constB=10000})) {constA=1, constB=0}
  n4 = Multiply(
   A=ComponentMask(None=TextureCoordinate()) {g=true}
   B=SphereMask(
    A=TextureCoordinate()
    B=Constant() {r=0.5}
    Radius=ScalarParameter() {defaultValue=0.5, group=Texture, parameterName=RainBlurred}
    Hardness=ScalarParameter() {group=Texture, parameterName=Hardness}
   )
  )
  n5 = Divide(
   A=Subtract(A=ActorPositionWS(), B=VectorParameter() {parameterName=BoxCenter, defaultValue=<rgba -55317.5 15954.5 6841 1>}.RGB)
   B=VectorParameter() {parameterName=BoxExtents, defaultValue=<rgba 18000 18000 6841 1>}.RGB
  )
  n6 = Constant() {r=0.5}
  n7 = Saturate(None=ComponentMask(None=Add(A=@n5, B=@n6)) {g=true, r=true})
  n8 = VertexColor()
  Multiply(
   A=Multiply(
    A=Multiply(A=@n4, B=@n3)
    B=TextureSample(UVs=@n7) {texture=/FNGameplayCues/Effects/Fort_Effects/Effects/Environment/Hermes/CrimeTown/Textures/T_CrimeTown_RainMask_Buildings_01}.A
   )
   B=Multiply(
    A=LinearInterpolate(Alpha=@n8.G) {constA=0.5}
    B=DynamicParameter() {paramNames [SizeStrecth, OpactiyPerSprite, OverallSpeed, Param4]}.OpactiyPerSprite
   )
  )
  n9 = ParticleColor()
  Multiply(A=@n9.RGB, B=@n9.RGB)
  NamedRerouteDeclaration(None=MaterialFunctionCall(
    Value_Morning=ScalarParameter() {defaultValue=1, parameterName="Value_Morning (S)"}
    Value_Day=ScalarParameter() {defaultValue=1, parameterName="Value_Day (S)"}
    Value_Evening=ScalarParameter() {defaultValue=1, parameterName="Value_Evening (S)"}
    Value_Night=ScalarParameter() {defaultValue=1, parameterName="Value_Night (S)"}
   ) {materialFunction=&Scalar_Time_of_Day}.Result) {name=TimeofDay, nodeColor=<rgba 1 0 0.5655 1>}
  NamedRerouteDeclaration(None=LinearInterpolate(A=Constant(), B=Constant() {r=1}, Alpha=NamedRerouteUsage())) {name=TOD_SpriteScale, nodeColor=<rgba 0.5234 1 0 1>}
  n10 = VectorParameter() {parameterName=PushUnderGround, defaultValue=<rgba 0 0 0 1>}
  n11 = ActorPositionWS()
  n12 = ScalarParameter() {defaultValue=2000, group=NewMove, parameterName=Scale}
  n13 = Multiply(A=Fmod(A=ComponentMask(None=@n11) {g=true, r=true}, B=@n12)) {constB=-1}
  n14 = Constant()
  n15 = TextureCoordinate()
  n16 = Multiply(
   A=ScalarParameter() {defaultValue=200, group=NewMove, parameterName=SizeStretch}
   B=DynamicParameter() {paramNames [SizeStrecth, OpactiyPerSprite, OverallSpeed, Param4]}.SizeStrecth
  )
  n17 = VectorParameter() {parameterName=Direction, defaultValue=<rgba 0 0 -1 1>}
  n18 = NamedRerouteUsage()
  n19 = ScalarParameter() {defaultValue=25, group=WPO, parameterName=maxSize}
  n20 = Multiply(A=@n19, B=VertexColor().G)
  n21 = Multiply(
   A=MaterialFunctionCall(Distance=ScalarParameter() {defaultValue=500, group=WPO, parameterName=NearCamShrinkDist}) {materialFunction=&Near_Camera_Fade}.Result
   B=MaterialFunctionCall(
    Fade_Start_Distance=ScalarParameter() {defaultValue=2000, parameterName="Fade_End_Distance (S)"}
    Fade_End_Distance=ScalarParameter() {defaultValue=2500, parameterName="Fade_Start_Distance (S)"}
   ) {materialFunction=&MF_Distance_SyncedFade_FX}.Result Inverted
  )
  n22 = Constant() {r=1}
  n23 = ComponentMask(None=ViewSize()) {g=true}
  n24 = MaterialFunctionCall() {materialFunction=&ScreenResolution}
  n25 = ComponentMask(None=@n24.Visible Resolution) {r=true}
  n26 = ComponentMask(None=@n24.Visible Resolution) {g=true}
  n27 = ScalarParameter() {defaultValue=1.33, parameterName=ScreenRes_Mininum}
  n28 = ScalarParameter() {defaultValue=2, parameterName=ScreenRes_Max}
  n29 = Subtract(A=Clamp(None=Divide(A=@n25, B=@n26), Min=@n27, Max=@n28) {maxDefault=2, minDefault=1.33}, B=@n27)
  n30 = Subtract(A=@n28, B=@n27)
  n31 = ScalarParameter() {defaultValue=0.08, group=WPO, parameterName=MinSizeatDistance_Full_ScreenRatio_Switch}
  n32 = ScalarParameter() {defaultValue=0.75, parameterName=SizeIncreaser_NintendoSwitch}
  n33 = LinearInterpolate(
   A=ScalarParameter() {defaultValue=0.08, group=WPO, parameterName=MinSizeatDistance_43_ScreenRatio}
   B=ShadingPathSwitch(Default=ScalarParameter() {defaultValue=0.08, group=WPO, parameterName=MinSizeatDistance_Full_ScreenRatio}, Forward=Divide(A=@n31, B=@n32))
   Alpha=Saturate(None=Divide(A=@n29, B=@n30))
  )
  n34 = Add(A=@n20, B=Divide(A=PixelDepth().R, B=Multiply(A=@n23, B=@n33)))
  n35 = ScalarParameter() {defaultValue=1, group=WPO, parameterName=UseMinimumSizeOnScreen?}
  n36 = Multiply(A=LinearInterpolate(A=@n20, B=@n19, Alpha=Constant()), B=Multiply(A=Clamp(None=@n21), B=LinearInterpolate(A=@n22, B=@n34, Alpha=@n35)))
  n37 = TextureSample() {samplerType=SAMPLERTYPE_Normal, texture=/Game/Environments/Asteria/Sets/NeonCity/Techart/Rain/T_RainLong_N}
  n38 = ScalarParameter() {defaultValue=0.5, parameterName=normalFlatness}
  n39 = Add(
   A=Multiply(
    A=MaterialFunctionCall(Float2=@n15) {materialFunction=&BreakOutFloat2Components}.G
    B=Multiply(A=@n16, B=@n17.RGB)
   )
   B=MaterialFunctionCall!(
    Center Location=WorldPosition().XYZ
    0-1 UVs
    XY scale=Multiply(A=@n18, B=@n36)
    OverridePivot
    "Normals (Optional)"=MaterialFunctionCall(Normal=@n37.RGB, Flatness=@n38) {materialFunction=&FlattenNormal}.Result
   ) {materialFunction=&Sprite}.World position offset
  )
  n40 = Multiply(A=Time(), B=ScalarParameter() {defaultValue=1.5, group=NewMove, parameterName=Speed}) {constB=0.4}
  n41 = Multiply(A=Multiply(A=Frac(None=@n40), B=@n12) {constB=2000}, B=@n17.RGB)
  n42 = Add(A=Add(A=AppendVector(A=@n13, B=@n14), B=Add(A=@n39, B=@n41)), B=Constant())
  n43 = Constant()
  n44 = Constant() {r=1}
  n45 = NamedRerouteUsage()
  Reroute(None=LinearInterpolate(A=Add(A=@n10.RGB, B=@n42), B=@n42, Alpha=LinearInterpolate(A=@n43, B=@n44, Alpha=@n45)))
 ]
}
