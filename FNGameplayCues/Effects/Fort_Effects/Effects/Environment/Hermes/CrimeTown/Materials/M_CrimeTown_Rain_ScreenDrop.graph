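# Additive, depth-test-free Niagara sprite material for rain drops on the screen in CrimeTown, drawn as bokeh tinted by particle color with a flattened round-drop normal.
# Three offset bokeh samples build a distortion that shifts the bokeh UVs.
# A top-down building texture, mapped by actor position within a box, provides a rain mask.
# A camera-direction mask is built, but the brightness lerp takes CameraUp_Mult at both ends, leaving CameraDown_Mult unused.
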Material /FNGameplayCues/Effects/Fort_Effects/Effects/Environment/Hermes/CrimeTown/Materials/M_CrimeTown_Rain_ScreenDrop
graph {
 properties {
  bDisableDepthTest=true
  bUsedWithNiagaraSprites=true
  blendMode=BLEND_Additive
  pixelDepthOffsetMode=PDOM_Legacy
  translucencyLightingMode=TLM_VolumetricPerVertexDirectional
  usedShadingModels=Default Lit
 }
 nodes [
  n1 = ComponentMask(None=MaterialFunctionCall() {materialFunction=&CameraDirectionVector}.Result) {b=true}
  LinearInterpolate(Alpha=Subtract(A=@n1) {constB=0.5})
  MaterialFunctionCall(
   Normal=TextureSample() {samplerType=SAMPLERTYPE_Normal, texture=/Game/Environments/Asteria/Sets/NeonCity/Techart/Rain/T_RainRound_N}.RGB
   Flatness=DynamicParameter() {defaultValue=<rgba 0.2 1 1 0>, paramNames [Distortion, FlattenNormal, CameraUp_Mult, CameraDown_Mult]}.FlattenNormal
  ) {materialFunction=&FlattenNormal}
  n2 = ParticleColor()
  n3 = TextureSample(UVs=NamedRerouteUsage()) {samplerType=SAMPLERTYPE_LinearColor, texture=/Game/TimeOfDay/TODM/BR/s11/ExposureAssets/FN_BokehS11}
  n4 = DynamicParameter() {defaultValue=<rgba 0.2 1 1 0>, paramNames [Distortion, FlattenNormal, CameraUp_Mult, CameraDown_Mult]}
  n5 = NamedRerouteUsage()
  n6 = LinearInterpolate(A=@n4.CameraUp_Mult, B=@n4.CameraUp_Mult, Alpha=OneMinus(None=Saturate(None=@n5)))
  Multiply(A=Multiply(A=Multiply(A=@n2.A, B=@n3.RGB), B=NamedRerouteUsage()), B=@n6)
  n7 = ParticleColor()
  Multiply(
   A=MaterialFunctionCall(Emissive=Multiply(A=@n7.RGB, B=@n3.RGB)) {materialFunction=&MF_FX_EmissiveAdjustments}.Result
   B=@n6
  )
  n8 = Subtract(A=ActorPositionWS(), B=VectorParameter() {parameterName=BoxCenter, defaultValue=<rgba -55317.5 15954.5 6841 1>}.RGB)
  n9 = VectorParameter() {parameterName=BoxExtents, defaultValue=<rgba 18000 18000 6841 1>}
  n10 = ComponentMask(None=Add(A=Divide(A=@n8, B=@n9.RGB), B=Constant() {r=0.5})) {g=true, r=true}
  NamedRerouteDeclaration(None=TextureSample(UVs=Saturate(None=@n10)) {texture=/FNGameplayCues/Effects/Fort_Effects/Effects/Environment/Hermes/CrimeTown/Textures/T_CrimeTown_RainMask_Buildings_01}.A) {name=Topdown texturemask, nodeColor=<rgba 0.6406 1 0 1>}
  n11 = Power(Base=@n1) {constExponent=16}
  NamedRerouteDeclaration(None=Abs(None=Saturate(None=@n11))) {name=CameraMask, nodeColor=<rgba 1 0.3984 0 1>}
  n12 = TextureCoordinate()
  n13 = Multiply(
   A=Constant3Vector() {constant=<rgba 1 0 0 0>}
   B=TextureSample(UVs=@n12) {samplerType=SAMPLERTYPE_LinearColor, texture=&FN_BokehS11}.R
  )
  n14 = Constant2Vector() {g=-0.03, r=-0.03}
  n15 = Multiply(
   A=Constant3Vector() {constant=<rgba 0 1 0 0>}
   B=TextureSample(UVs=Add(A=@n12, B=@n14)) {samplerType=SAMPLERTYPE_LinearColor, texture=&FN_BokehS11}.R
  )
  n16 = Constant3Vector() {constant=<rgba 0 0 1 0>}
  n17 = TextureSample(UVs=Add(A=@n12, B=Constant2Vector() {g=0.03, r=0.03})) {samplerType=SAMPLERTYPE_LinearColor, texture=&FN_BokehS11}
  n18 = ComponentMask(None=Add(A=Add(A=@n13, B=@n15), B=Multiply(A=@n16, B=@n17.R))) {g=true, r=true}
  n19 = DynamicParameter() {defaultValue=<rgba 0.2 1 1 0>, paramNames [Distortion, FlattenNormal, CameraUp_Mult, CameraDown_Mult]}
  NamedRerouteDeclaration(None=Add(A=Multiply(A=@n18, B=@n19.Distortion), B=TextureCoordinate())) {name=BokehUV, nodeColor=<rgba 0.9277 0.7344 1 1>}
 ]
}
