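# Niagara rain system for CrimeTown that surrounds the camera with layered streaks, runoff from objects and splashes.
# GPU emitters share a time-of-day spawn curve that stops rain through daytime; a simpler mesh emitter serves mobile.
# Runoff and splashes snap to surfaces via depth buffer and distance fields, stopping indoors.
# Building mask textures and kill volumes carve out dry areas, while several layers inherit local player velocity.
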NiagaraSystem /FNGameplayCues/Effects/Fort_Effects/Effects/Environment/Hermes/CrimeTown/NS_CrimeTown_Rain_Camera
dependencies {
 usedDataInterfaces [
  &NiagaraDataInterfaceCurve
  /Script/Niagara.NiagaraDataInterfaceActorComponent
  /Script/Niagara.NiagaraDataInterfaceTexture
  &NiagaraDataInterfaceCollisionQuery
  &NiagaraDataInterfaceLandscape
  &NiagaraDataInterfaceGBuffer
  &NiagaraDataInterfaceColorCurve
 ]
 usedDynamicInputs [&ColorFromCurve]
 usedModules [
  &SystemState
  &self:SystemSpawnScript.SystemScriptSource.SystemScriptGraph.NiagaraNodeAssignment_1.SetVariables_#1
  &self:SystemSpawnScript.SystemScriptSource.SystemScriptGraph.NiagaraNodeAssignment_0.SetVariables_#2
  &self:SystemSpawnScript.SystemScriptSource.SystemScriptGraph.NiagaraNodeAssignment_2.SetVariables_#3
 ]
}
emitters {
 RainFromObjects {
  simTarget=GPUComputeSim
  emitterUpdateScript [
   &EmitterState()
   &SpawnRate(
    SpawnRate=&Multiply_Float(
     A=&FloatFromCurve(
      FloatCurve=DataInterface() {propertyValues {Curve {keys [<at 0.08 1>,<at 0.185 0>,<at 0.9375 0>,<at 1 1>]}}}
      CurveIndex=Linked() {linkedVariable {name=NPC.FortniteNPC.TimeOfDay_0-1, type=NiagaraFloat}}
      Scale Curve=$User.Rain_SpawnRateScale
     )
     B=&Multiply_Float(A=&Lerp_Float(A=1, B=0, Alpha=$NPC.FortniteNPC.IsPlayerIndoors), B=5000)
    )
    Use Spawn Probability=false
   )
  ]
  particleSpawnScript [
   &InitializeParticle(
    Lifetime Max=1
    Lifetime Min=0.5
    Lifetime Mode=::NewEnumerator1
    Sprite Rotation Mode=::NewEnumerator1
    Sprite Size Max=<25 20>
    Sprite Size Min=<25 15>
    Sprite Size Mode=::NewEnumerator2
    UsePositionOffset=false
    Write Material Random=true
   ) {module=InitializeParticle001}
   &AddVelocity(Velocity=&RandomRangeVector(Minimum=<0 0 -50>, Maximum=<0 0 -50>))
   &SphereLocation(
    MaskSpawnGroup=false
    Non Uniform Scale=<1 1 0.7>
    Offset=<0 0 1500>
    Override Seed=false
    Sphere Radius=5000
    Surface Only=true
    Surface Only Band Thickness=0.5
    UseNonUniformScale=true
    UseOffset=true
    UseSphereOrientation=false
   )
   /Niagara/Modules/Math/MoveToNearestDistanceFieldSurface_GPU(Kill by Distance=800, bKillByDistance=true)
   /SetVariables(Particles.InitialPos:NiagaraPosition=$Particles.Position) {isSetParameters=true}
   /Niagara/Modules/Math/PlaceParticlesOnDepthBuffer_GPU(Max Z Depth=1500, Module.Use Max Z Depth=true)
   &KillParticles(Kill Particles=&SetBoolByFloatComparison(
     A=&VectorLength(Vector=&Subtract_Vector(A=$Particles.InitialPos, B=Linked() {linkedVariable {name=Particles.Position, type=NiagaraPosition}}))
     B=1500
    ))
   &KillParticles(Kill Particles=&Boolean_NotOperation(Bool=$Output.PlaceParticlesOnDepthBuffer_GPU.ValidSample)) {module=KillParticles001}
   &SubUVAnimation(End Frame=3, SubUV Animation Mode:ENiagara_SubUVLookupMode=::NewEnumerator2)
   &CameraOffset(Camera Offset Amount=-10)
   &self:ScratchModule_08(Angle=180, BlendDirection=1, Normal:Vector3f=<0 0 1>, PickCeilingOrFloor=true, Velocity=-200) {module=ScratchModule_08}
   &self:ViewMask(Border Percent Max=25, Border Percent Min=15) {module=ViewMask001}
   &self:RainTextureMasking(
    AlphaThreshold=0.5
    BoxCenter:NiagaraPosition=<-50828 29472 9916>
    BoxExtents:Vector3f=<32500 32500 1>
    Position:NiagaraPosition=$Particles.Position
    RejectionTexture=DataInterface() {propertyValues {
      Texture=/Game/Environments/Asteria/Sets/NeonCity/FX/Textures/T_NeonCity_Rain_Mask_01
      TextureUserParameter {parameter {name=None, typeDefHandle {registeredTypeIndex=2}, varData []}}
     }}
   ) {module=RainTextureMasking}
   &KillParticles(Kill Particles=&SetBoolByFloatComparison(A=0.2, B=$Output.ViewMask001.ViewMask)) {module=KillParticles002}
   &KillParticlesInVolume(
    Box Size:Vector3f=<3000 3000 800>
    Kill Shape=::NewEnumerator1
    Volume Origin:Vector3f=&AddVectorToPosition(Position=<-50410.4922 20481.1484 4000>, Vector=<0 0 0>)
   )
  ]
  particleUpdateScript [
   &ParticleState()
   &self:ViewMask(Border Percent Max=15, Border Percent Min=5) {module=ViewMask}
   &RegionMask(Inner Float=0, Outer Float=2, Region Falloff=0, Region Radius:NiagaraFloat=$User.Rain_MaskInside_Radius, UseOffset=false, UseRotation=false) {module=RegionMask001}
   &ScaleSpriteSize(Uniform Curve Sprite Scale=DataInterface() {propertyValues {Curve {keys [{tangentMode=RCTM_User, time=0, value=0.85}, {tangentMode=RCTM_User, time=1, value=1}]}}})
   &GravityForce(Gravity=<0 0 -400>)
   &SolveForcesAndVelocity(Clamp Velocity=false, Limit Acceleration=false, Speed Limit=500)
   &ScaleSpriteSizeBySpeed(
    Max Scale Factor=<0.1 5>
    Min Scale Factor=<0.05 1>
    Source Velocity:Vector3f=&Multiply_Vector(A=<0.05 0.05 1>, B=$Particles.Velocity)
   )
   &ScaleColor(
    Scale Alpha=&Multiply_Float(
     A=&Multiply_Float(
      A=&Multiply_Float(
       A=&Multiply_Float(
        A=&FloatFromCurve(
         FloatCurve=DataInterface() {propertyValues {Curve {keys [
             {interpMode=RCIM_Cubic, tangentMode=RCTM_User, time=0, value=0}
             {interpMode=RCIM_Cubic, tangentMode=RCTM_User, time=0.5, value=1}
             {tangentMode=RCTM_User, time=1, value=0}
            ]}}}
         Scale Curve=0.25
        )
        B=$System.IndoorAlpha
       )
       B=$Particles.RegionMask001.Float
      )
      B=$User.Rain_AlphaMult
     )
     B=&Lerp_Float(A=0, B=1, Alpha=$Output.ViewMask.ViewMask)
    )
    Scale RGB=&MakeVectorFromLinearColorRGB(Color=&MultiplyLinearColor(Linear Color=$User.Rain_NonLitColor, Scale Factor=&MakeVector4_FromFloat(Float=$User.Rain_ColorMult)))
    ScaleA=true
    ScaleRGB=true
   )
   /SetVariables(Particles.HighEPIC COLOR:LinearColor=&MakeLinearColorFromVectorAndFloat(
     "Vector (RGB)"=&Multiply_Vector(
      A=&MakeVectorFromLinearColorRGB(Color=$Particles.Initial.Color)
      B=&VectorFromFloat(Value=&Multiply_Float(A=2, B=$User.Rain_ColorMult))
     )
     "Float (Alpha)"=&Multiply_Float(
      A=&Multiply_Float(
       A=&Multiply_Float(
        A=&FloatFromCurve(FloatCurve=DataInterface() {propertyValues {Curve {keys [
             {interpMode=RCIM_Cubic, tangentMode=RCTM_User, time=0.2, value=0}
             {interpMode=RCIM_Cubic, tangentMode=RCTM_User, time=0.5, value=1}
             {tangentMode=RCTM_User, time=1, value=0}
            ]}}})
        B=$System.IndoorAlpha
       )
       B=$Particles.RegionMask001.Float
      )
      B=$User.Rain_AlphaMult
     )
    )) {isSetParameters=true}
  ]
  rendererClasses [&NiagaraSpriteRendererProperties]
  renderers [
   {rendererClass=&NiagaraSpriteRendererProperties}
   {rendererClass=&NiagaraSpriteRendererProperties, rendererIndex=1}
  ]
 }
 Rain_Close {
  simTarget=GPUComputeSim
  emitterUpdateScript [
   &EmitterState(Life Cycle Mode=::NewEnumerator1, Loop Duration=2, UseLoopDelay=true)
   &SpawnRate(
    SpawnRate=&Multiply_Float(
     A=&FloatFromCurve(
      FloatCurve=DataInterface() {propertyValues {Curve {keys [<at 0.08 1>,<at 0.185 0>,<at 0.9375 0>,<at 1 1>]}}}
      CurveIndex=Linked() {linkedVariable {name=NPC.FortniteNPC.TimeOfDay_0-1, type=NiagaraFloat}}
      Scale Curve=$User.Rain_SpawnRateScale
     )
     B=15000
    )
    Use Spawn Probability=false
   )
  ]
  particleSpawnScript [
   &InitializeParticle(
    Lifetime Max=1
    Lifetime Min=0.6
    Lifetime Mode=::NewEnumerator1
    Sprite Size Max=<2 2>
    Sprite Size Min=<2 2>
    Sprite Size Mode=::NewEnumerator2
    UsePositionOffset=false
    Write Material Random=true
   ) {module=InitializeParticle001}
   &self:ScratchModule_07(New Actor Component Interface=DataInterface() {propertyValues {
      LocalPlayerIndex=0
      SourceActorRef=null
      SourceMode=LocalPlayer
      bRequireCurrentFrameData=true
      ActorOrComponentParameter {parameter {name=None, typeDefHandle {registeredTypeIndex=1}, varData []}}
     }}) {module=ScratchModule_07}
   &AddVelocity(
    Coordinate Space=::World
    Velocity=&Add_Vector(
     A=&Multiply_VectorByFloat(Vector=&RandomRangeVector(Minimum=<0 0 -800>, Maximum=<0 0 -1000>), Float=$User.Rain_FallingSpeed)
     B=&Multiply_Vector(A=$Output.ScratchModule_07.Velocity, B=$User.Rain_InheritVel)
    )
   )
   /SetVariables() {isSetParameters=true, module=SetVariables_#4}
   &ShapeLocation(
    Cylinder Height=15000
    Cylinder Radius=2500
    MaskSpawnGroup=false
    Non Uniform Scale=<1 1 0.2>
    Offset=<0 0 1250>
    Offset Mode=::NewEnumerator0
    Override Seed=false
    Shape Primitive=::NewEnumerator1
    Surface Only=true
    Surface Only Band Thickness=1
    Use Endcaps In Surface Only Mode=false
    UseOwnerScale=true
   )
   &ShapeLocation(
    Box Size=<18000 18000 100>
    MaskSpawnGroup=false
    Override Seed=false
    Shape Primitive=::NewEnumerator2
    Surface Only=false
    UseOwnerScale=true
   ) {enabled=false, module=ShapeLocation001}
   &self:RainTextureMasking(
    AlphaThreshold=0.9
    BoxCenter:NiagaraPosition=<-55317.5 15954.5 6841>
    BoxExtents:Vector3f=<18000 18000 1>
    Position:NiagaraPosition=$Particles.Position
    RejectionTexture=DataInterface() {propertyValues {
      Texture=/FNGameplayCues/Effects/Fort_Effects/Effects/Environment/Hermes/CrimeTown/Textures/T_CrimeTown_RainMask_Buildings_01
      TextureUserParameter {parameter {name=None, typeDefHandle {registeredTypeIndex=2}, varData []}}
     }}
   ) {module=RainTextureMasking}
   &KillParticlesInVolume(
    Box Size:Vector3f=<3000 3000 400>
    Kill Shape=::NewEnumerator1
    Volume Origin:Vector3f=&AddVectorToPosition(Position=<-50410.4922 20481.1484 4000>, Vector=<0 0 0>)
   )
  ]
  particleUpdateScript [
   &ParticleState()
   &WindForce(
    Scale=0.1
    Turbulence Flow Loop Duration=1
    Wind Speed=&Lerp_Vector(A=<0 0 0>, B=<0 0 0>, Alpha=$System.StormIntensity)
   ) {enabled=false}
   &AerodynamicDrag(Aerodynamic Drag=0.01, Use Linear Drag=true, Use Rotational Drag=true) {enabled=false}
   &WindForce(Wind Speed=&Lerp_Vector(A=<600 0 0>, B=<800 0 0>, Alpha=$System.StormIntensity)) {enabled=false, module=WindForce001}
   &SolveForcesAndVelocity(Clamp Velocity=false, Limit Acceleration=false)
   &ScaleSpriteSizeBySpeed(
    Max Scale Factor=<3 50>
    Min Scale Factor=<2 35>
    Source Velocity:Vector3f=&Multiply_Vector(A=<0.05 0.05 1>, B=$Particles.Velocity)
   )
   &AlignSpriteToMeshOrientation(SetAlignment=true, SetFacing=true)
   &ScaleColor(
    Scale Alpha=&Multiply_Float(
     A=&Multiply_Float(
      A=&FloatFromCurve(
       FloatCurve=DataInterface() {propertyValues {Curve {keys [
           {interpMode=RCIM_Cubic, tangentMode=RCTM_User, time=0, value=0}
           {interpMode=RCIM_Cubic, tangentMode=RCTM_User, time=0.5, value=1}
           {tangentMode=RCTM_User, time=1, value=0}
          ]}}}
       Scale Curve=0.35
      )
      B=$System.IndoorAlpha
     )
     B=$User.Rain_AlphaMult
    )
    Scale RGB=&MakeVectorFromLinearColorRGB(Color=&MultiplyLinearColor(Linear Color=$User.Rain_NonLitColor, Scale Factor=&MakeVector4_FromFloat(Float=$User.Rain_ColorMult)))
    ScaleA=true
    ScaleRGB=true
   )
   /SetVariables(Particles.Low COLOR:LinearColor=&MakeLinearColorFromVectorAndFloat(
     "Vector (RGB)"=&MakeVectorFromLinearColorRGB(Color=&MultiplyLinearColor(
       Linear Color=$User.Rain_NonLitColor_Low
       Scale Factor=&MakeVector4_FromFloat(Float=$User.Rain_ColorMult)
      ))
     "Float (Alpha)"=&Multiply_Float(
      A=&Multiply_Float(
       A=&Multiply_Float(
        A=&FloatFromCurve(FloatCurve=DataInterface() {propertyValues {Curve {keys [
             {interpMode=RCIM_Cubic, tangentMode=RCTM_User, time=0, value=0}
             {interpMode=RCIM_Cubic, tangentMode=RCTM_User, time=0.5, value=1}
             {tangentMode=RCTM_User, time=1, value=0}
            ]}}})
        B=$System.IndoorAlpha
       )
       B=1
      )
      B=$User.Rain_AlphaMult
     )
    )) {isSetParameters=true}
   /SetVariables(Particles.HighEPIC COLOR:LinearColor=&MakeLinearColorFromVectorAndFloat(
     "Vector (RGB)"=&Multiply_Vector(
      A=&MakeVectorFromLinearColorRGB(Color=$Particles.Initial.Color)
      B=&VectorFromFloat(Value=&Multiply_Float(A=2, B=$User.Rain_ColorMult))
     )
     "Float (Alpha)"=&Multiply_Float(
      A=&Multiply_Float(
       A=&Multiply_Float(
        A=&FloatFromCurve(
         FloatCurve=DataInterface() {propertyValues {Curve {keys [
             {interpMode=RCIM_Cubic, tangentMode=RCTM_User, time=0, value=0}
             {interpMode=RCIM_Cubic, tangentMode=RCTM_User, time=0.5, value=1}
             {tangentMode=RCTM_User, time=1, value=0}
            ]}}}
         Scale Curve=0.25
        )
        B=$System.IndoorAlpha
       )
       B=1
      )
      B=$User.Rain_AlphaMult
     )
    )) {isSetParameters=true}
   /SetVariables(Particles.Rain_Low_Scale:Vector2f=&Multiply_Vector2D(A=$Particles.SpriteSize, B=$System.Rain_Low_Scale)) {isSetParameters=true}
  ]
  rendererClasses [&NiagaraSpriteRendererProperties]
  renderers [
   {rendererClass=&NiagaraSpriteRendererProperties}
   {rendererClass=&NiagaraSpriteRendererProperties, rendererIndex=1}
   {rendererClass=&NiagaraSpriteRendererProperties, rendererIndex=2}
  ]
 }
 Rain_Far {
  simTarget=GPUComputeSim
  emitterUpdateScript [
   &EmitterState(Life Cycle Mode=::NewEnumerator1, Loop Duration=2, UseLoopDelay=true)
   &SpawnRate(
    SpawnRate=&Multiply_Float(
     A=&FloatFromCurve(
      FloatCurve=DataInterface() {propertyValues {Curve {keys [<at 0.08 1>,<at 0.185 0>,<at 0.9375 0>,<at 1 1>]}}}
      CurveIndex=Linked() {linkedVariable {name=NPC.FortniteNPC.TimeOfDay_0-1, type=NiagaraFloat}}
      Scale Curve=$User.Rain_SpawnRateScale
     )
     B=25
    )
    Use Spawn Probability=false
   )
  ]
  particleSpawnScript [
   &InitializeParticle(
    Color=<rgba 0.004 0.1274 0.5972 1>
    Lifetime Max=4
    Lifetime Min=3
    Lifetime Mode=::NewEnumerator1
    Sprite Size Max=<6000 50000>
    Sprite Size Min=<4000 25000>
    Sprite Size Mode=::NewEnumerator2
    UsePositionOffset=false
    Write Material Random=true
   ) {module=InitializeParticle001}
   &AddVelocity(Velocity=&Multiply_VectorByFloat(Vector=&RandomRangeVector(Minimum=<0 0 -35>, Maximum=<0 0 -50>), Float=$User.Rain_FallingSpeed))
   /SetVariables() {isSetParameters=true, module=SetVariables_#4}
   &Color(Color=&MakeLinearColorFromVectorAndFloat("Vector (RGB)"=<0.6281 0.5529 0.974>, "Float (Alpha)"=&ClampFloat(Float=1.2))) {enabled=false}
   &ShapeLocation(
    MaskSpawnGroup=false
    Override Seed=false
    Ring / Disc Distribution Mode:ENiagaraSphereDistributionMode=::NewEnumerator1
    Ring Radius=15000
    Shape Primitive=::NewEnumerator4
    UseOwnerScale=true
   )
   &KillParticlesInVolume(Origin Offset:Vector3f=<0 0 -100>, Sphere Radius=3500)
  ]
  particleUpdateScript [
   &ParticleState()
   &SolveForcesAndVelocity(Clamp Velocity=false, Limit Acceleration=false)
   &ScaleColor(
    Scale Alpha=&FloatFromCurve(
     FloatCurve=DataInterface() {propertyValues {Curve {keys [
         {interpMode=RCIM_Cubic, tangentMode=RCTM_User, time=0, value=0}
         {interpMode=RCIM_Cubic, tangentMode=RCTM_User, time=0.3, value=1}
         <at 1 0 RCIM_Cubic>
        ]}}}
     Scale Curve=0.01
    )
    Scale RGB=&MakeVectorFromLinearColorRGB(Color=$Particles.Initial.Color)
    ScaleA=true
    ScaleRGB=true
   )
   &DynamicMaterialParameters(
    Index 0 Param 1=&RandomRangeFloat(Minimum=0, Maximum=5)
    Index 0 Param 2=&RandomRangeFloat(Minimum=0, Maximum=5)
    Index 0 Param 3=&RandomRangeFloat(Minimum=1, Maximum=1)
    Index 0 Param 4=&RandomRangeFloat(Minimum=1, Maximum=5)
    Index 1 Param 1=4
    Index 1 Param 2=2
    Module.Param0WriteEnabled=true
    Module.Param0WriteEnabled1=true
    Module.Param1WriteEnabled=true
    Module.Param1WriteEnabled1=true
    Module.Param2WriteEnabled=true
    Module.Param2WriteEnabled1=true
    Module.Param3WriteEnabled=true
    Module.Param3WriteEnabled1=true
    Write Parameter Index 1=true
   )
   /SetVariables(Particles.HighEPIC COLOR:LinearColor=&MakeLinearColorFromVectorAndFloat(
     "Vector (RGB)"=&Multiply_Vector(
      A=&MakeVectorFromLinearColorRGB(Color=$Particles.Initial.Color)
      B=&VectorFromFloat(Value=&Multiply_Float(A=1, B=$User.Rain_ColorMult))
     )
     "Float (Alpha)"=&Multiply_Float(
      A=&Multiply_Float(
       A=&FloatFromCurve(
        FloatCurve=DataInterface() {propertyValues {Curve {keys [
            {interpMode=RCIM_Cubic, tangentMode=RCTM_User, time=0, value=0}
            {interpMode=RCIM_Cubic, tangentMode=RCTM_User, time=0.5, value=1}
            {tangentMode=RCTM_User, time=1, value=0}
           ]}}}
        Scale Curve=5
       )
       B=1
      )
      B=$User.Rain_AlphaMult
     )
    )) {isSetParameters=true}
  ]
  rendererClasses [&NiagaraSpriteRendererProperties]
  renderers [
   {rendererClass=&NiagaraSpriteRendererProperties}
   {rendererClass=&NiagaraSpriteRendererProperties, rendererIndex=1}
  ]
 }
 Rain_Mid {
  simTarget=GPUComputeSim
  emitterUpdateScript [
   &EmitterState(Life Cycle Mode=::NewEnumerator1, Loop Duration=2, UseLoopDelay=true)
   &SpawnRate(
    SpawnRate=&Multiply_Float(
     A=&FloatFromCurve(
      FloatCurve=DataInterface() {propertyValues {Curve {keys [<at 0.08 1>,<at 0.185 0>,<at 0.9375 0>,<at 1 1>]}}}
      CurveIndex=Linked() {linkedVariable {name=NPC.FortniteNPC.TimeOfDay_0-1, type=NiagaraFloat}}
      Scale Curve=$User.Rain_SpawnRateScale
     )
     B=4000
    )
    Use Spawn Probability=false
   )
  ]
  particleSpawnScript [
   &InitializeParticle(
    Lifetime Max=0.8
    Lifetime Min=0.5
    Lifetime Mode=::NewEnumerator1
    Sprite Size Max=<3.75 4.5>
    Sprite Size Min=<3 3.3>
    Sprite Size Mode=::NewEnumerator2
    UsePositionOffset=false
    Write Material Random=true
   ) {module=InitializeParticle001}
   &self:ScratchModule_07(New Actor Component Interface=DataInterface() {propertyValues {
      LocalPlayerIndex=0
      SourceActorRef=null
      SourceMode=LocalPlayer
      bRequireCurrentFrameData=true
      ActorOrComponentParameter {parameter {name=None, typeDefHandle {registeredTypeIndex=1}, varData []}}
     }}) {module=ScratchModule_07}
   &AddVelocity(
    Coordinate Space=::World
    Velocity=&Add_Vector(
     A=&Multiply_VectorByFloat(Vector=&RandomRangeVector(Minimum=<0 0 -1200>, Maximum=<0 0 -1400>), Float=$User.Rain_FallingSpeed)
     B=&Multiply_Vector(A=$Output.ScratchModule_07.Velocity, B=$User.Rain_InheritVel)
    )
   )
   /SetVariables() {isSetParameters=true, module=SetVariables_#4}
   &Color(Color=&MakeLinearColorFromVectorAndFloat("Vector (RGB)"=<0.6281 0.5529 0.974>, "Float (Alpha)"=&ClampFloat(Float=1.2))) {enabled=false}
   &KillParticlesInVolume(Kill Volume Enabled=false, Origin Offset:Vector3f=<0 0 -100>, Sphere Radius=5000)
   &ShapeLocation(
    Cylinder Height=2500
    Cylinder Radius=6000
    MaskSpawnGroup=false
    Non Uniform Scale=<1 1 0.2>
    Offset=<0 0 1250>
    Offset Mode=::NewEnumerator0
    Override Seed=false
    Shape Primitive=::NewEnumerator1
    Surface Only=true
    Surface Only Band Thickness=0.7
    Use Endcaps In Surface Only Mode=false
    UseOwnerScale=true
   )
   &self:RainTextureMasking(
    AlphaThreshold=0.5
    BoxCenter:NiagaraPosition=<-55317.5 15954.5 6841>
    BoxExtents:Vector3f=<18000 18000 1>
    Position:NiagaraPosition=$Particles.Position
    RejectionTexture=DataInterface() {propertyValues {
      Texture=&T_CrimeTown_RainMask_Buildings_01
      TextureUserParameter {parameter {name=None, typeDefHandle {registeredTypeIndex=2}, varData []}}
     }}
   ) {module=RainTextureMasking}
   /Niagara/Modules/Update/Textures/SampleTexture(
    Enable Sample Rejection=true
    EnableFalloff=false
    Generate UV Using Importance Sampling=true
    Texture=DataInterface() {propertyValues {
      Texture=&T_CrimeTown_RainMask_Buildings_01
      TextureUserParameter {parameter {name=None, typeDefHandle {registeredTypeIndex=2}, varData []}}
     }}
    Transform UV Into Position=true
   ) {enabled=false}
   &KillParticlesInVolume(
    Box Size:Vector3f=<3000 3000 400>
    Kill Shape=::NewEnumerator1
    Volume Origin:Vector3f=&AddVectorToPosition(Position=<-50410.4922 20481.1484 4000>, Vector=<0 0 0>)
   ) {module=KillParticlesInVolume001}
  ]
  particleUpdateScript [
   &ParticleState()
   &WindForce(Frequency=5, Scale=0.1, Wind Speed=&Lerp_Vector(A=<0 0 0>, B=<0 0 0>, Alpha=$System.StormIntensity)) {enabled=false}
   &AerodynamicDrag(
    Aerodynamic Drag=0.01
    Drag Shape Curve=DataInterface() {propertyValues {Curve {keys [
        {interpMode=RCIM_Cubic, tangentMode=RCTM_User, time=0, value=0.1}
        {arriveTangent=2.0799, interpMode=RCIM_Cubic, leaveTangent=2.0799, tangentMode=RCTM_User, time=1, value=1}
       ]}}}
    Lift Shape Curve=DataInterface() {propertyValues {Curve {keys [
        {interpMode=RCIM_Cubic, tangentMode=RCTM_User, time=0, value=0}
        {interpMode=RCIM_Cubic, tangentMode=RCTM_User, time=0.5, value=1}
        {tangentMode=RCTM_User, time=1, value=0}
       ]}}}
    Use Linear Drag=true
    Use Rotational Drag=true
   ) {enabled=false}
   &SolveForcesAndVelocity(Clamp Velocity=false, Limit Acceleration=false)
   &AlignSpriteToMeshOrientation(SetAlignment=true, SetFacing=true)
   &ScaleSpriteSizeBySpeed(
    Max Scale Factor=<2 50>
    Min Scale Factor=<1 35>
    Source Velocity:Vector3f=&Multiply_Vector(A=<0.05 0.05 1>, B=$Particles.Velocity)
   )
   &ScaleColor(
    Scale Alpha=&Multiply_Float(
     A=&Multiply_Float(
      A=&FloatFromCurve(FloatCurve=DataInterface() {propertyValues {Curve {keys [
           {interpMode=RCIM_Cubic, tangentMode=RCTM_User, time=0, value=0}
           {interpMode=RCIM_Cubic, tangentMode=RCTM_User, time=0.5, value=1}
           {tangentMode=RCTM_User, time=1, value=0}
          ]}}})
      B=$System.IndoorAlpha
     )
     B=$User.Rain_AlphaMult
    )
    Scale RGB=&MakeVectorFromLinearColorRGB(Color=&MultiplyLinearColor(Linear Color=$User.Rain_NonLitColor, Scale Factor=&MakeVector4_FromFloat(Float=$User.Rain_ColorMult)))
    ScaleA=true
    ScaleRGB=true
   )
   /SetVariables(Particles.HighEPIC COLOR:LinearColor=&MakeLinearColorFromVectorAndFloat(
     "Vector (RGB)"=&Multiply_Vector(
      A=&MakeVectorFromLinearColorRGB(Color=$Particles.Initial.Color)
      B=&VectorFromFloat(Value=&Multiply_Float(A=1.5, B=$User.Rain_ColorMult))
     )
     "Float (Alpha)"=&Multiply_Float(
      A=&Multiply_Float(
       A=&Multiply_Float(
        A=&FloatFromCurve(FloatCurve=DataInterface() {propertyValues {Curve {keys [
             {interpMode=RCIM_Cubic, tangentMode=RCTM_User, time=0, value=0}
             {interpMode=RCIM_Cubic, tangentMode=RCTM_User, time=0.5, value=1}
             {tangentMode=RCTM_User, time=1, value=0}
            ]}}})
        B=$System.IndoorAlpha
       )
       B=1
      )
      B=$User.Rain_AlphaMult
     )
    )) {isSetParameters=true}
  ]
  rendererClasses [&NiagaraSpriteRendererProperties]
  renderers [
   {rendererClass=&NiagaraSpriteRendererProperties}
   {rendererClass=&NiagaraSpriteRendererProperties, rendererIndex=1}
  ]
 }
 Rain_Mobile {
  emitterUpdateScript [
   &EmitterState(Life Cycle Mode=::NewEnumerator1, Loop Duration=2, UseLoopDelay=true)
   &SpawnRate(SpawnRate=50, Use Spawn Probability=false)
  ]
  particleSpawnScript [
   &InitializeParticle(
    Lifetime Max=1.5
    Lifetime Mode=::NewEnumerator1
    Mesh Scale Max=<10 10 25>
    Mesh Scale Min=<5 5 25>
    Mesh Scale Mode=::NewEnumerator2
    Sprite Size Max=<3.75 4.5>
    Sprite Size Min=<3 3.3>
    Sprite Size Mode=::NewEnumerator2
    UsePositionOffset=false
    Write Material Random=true
   ) {module=InitializeParticle001}
   &AddVelocity(
    Coordinate Space=::World
    Velocity=&Add_Vector(
     A=&Multiply_VectorByFloat(Vector=&RandomRangeVector(Minimum=<0 0 -800>, Maximum=<0 0 -1200>), Float=$User.Rain_FallingSpeed)
     B=<0 0 1>
    )
   )
   /SetVariables() {isSetParameters=true, module=SetVariables_#4}
   &Color(Color=&MakeLinearColorFromVectorAndFloat("Vector (RGB)"=<0.6281 0.5529 0.974>, "Float (Alpha)"=&ClampFloat(Float=1.2))) {enabled=false}
   &ShapeLocation(
    Cylinder Height=5000
    Cylinder Radius=2500
    MaskSpawnGroup=false
    Non Uniform Scale=<1 1 0.2>
    Offset=<0 0 2500>
    Offset Mode=::NewEnumerator0
    Override Seed=false
    Shape Primitive=::NewEnumerator1
    Surface Only=true
    Surface Only Band Thickness=0.7
    Use Endcaps In Surface Only Mode=false
    UseOwnerScale=true
   )
   &InitialMeshOrientation(Rotation=&RandomRangeVector(Minimum=<0 0 0>, Maximum=<0 0 1>), Use Rotation Vector=true)
  ]
  particleUpdateScript [
   &ParticleState()
   &GravityForce(Gravity=<0 0 -300>)
   &SolveForcesAndVelocity(Clamp Velocity=true, Limit Acceleration=false, Speed Limit=5000)
   &AlignSpriteToMeshOrientation(SetAlignment=true, SetFacing=true)
   &ScaleColor(
    Scale Alpha=&Multiply_Float(
     A=&Multiply_Float(
      A=&FloatFromCurve(
       FloatCurve=DataInterface() {propertyValues {Curve {keys [
           {interpMode=RCIM_Cubic, tangentMode=RCTM_User, time=0, value=0}
           {interpMode=RCIM_Cubic, tangentMode=RCTM_User, time=0.5, value=1}
           {tangentMode=RCTM_User, time=1, value=0}
          ]}}}
       Scale Curve=0.35
      )
      B=$System.IndoorAlpha
     )
     B=$User.Rain_AlphaMult
    )
    Scale RGB=&Multiply_VectorByFloat(Vector=<1 1 1>, Float=$User.Rain_ColorMult)
    ScaleA=true
    ScaleRGB=true
   )
   &DynamicMaterialParameters(
    Index 0 Param 1=&Multiply_Float(
     A=&RandomRangeFloat(Minimum=0.7, Maximum=2)
     B=&FloatFromCurve(FloatCurve=DataInterface() {propertyValues {Curve {keys [
          {interpMode=RCIM_Cubic, tangentMode=RCTM_User, time=0, value=1}
          {interpMode=RCIM_Cubic, tangentMode=RCTM_User, time=0.5, value=1}
          {tangentMode=RCTM_User, time=1, value=1}
         ]}}})
    )
    Index 0 Param 2=&RandomRangeFloat(Minimum=1, Maximum=2)
    Index 0 Param 3=&RandomRangeFloat(Minimum=0.1, Maximum=0.2)
    Module.Param0WriteEnabled=true
    Module.Param1WriteEnabled=true
    Module.Param2WriteEnabled=true
    Module.Param3WriteEnabled=true
   )
  ]
  rendererClasses [&NiagaraMeshRendererProperties]
  renderers [{rendererClass=&NiagaraMeshRendererProperties, rendererIndex=0}]
 }
 Rain_VeryClose {
  simTarget=GPUComputeSim
  emitterUpdateScript [
   &EmitterState(Life Cycle Mode=::NewEnumerator1, Loop Duration=2, UseLoopDelay=true)
   &SpawnRate(
    SpawnRate=&Multiply_Float(
     A=&FloatFromCurve(
      FloatCurve=DataInterface() {propertyValues {Curve {keys [<at 0.08 1>,<at 0.185 0>,<at 0.9375 0>,<at 1 1>]}}}
      CurveIndex=Linked() {linkedVariable {name=NPC.FortniteNPC.TimeOfDay_0-1, type=NiagaraFloat}}
      Scale Curve=$User.Rain_SpawnRateScale
     )
     B=&Multiply_Float(A=&Lerp_Float(A=1, B=0.25, Alpha=$NPC.FortniteNPC.IsPlayerIndoors), B=8000)
    )
    Use Spawn Probability=false
   )
   /SetVariables(Emitter.VolumeRadius=250) {isSetParameters=true}
  ]
  particleSpawnScript [
   &InitializeParticle(
    Lifetime Max=1
    Lifetime Min=0.6
    Lifetime Mode=::NewEnumerator1
    Sprite Size Max=<1.5 3>
    Sprite Size Min=<1 2.5>
    Sprite Size Mode=::NewEnumerator2
    UsePositionOffset=false
    Write Material Random=true
   ) {module=InitializeParticle001}
   &self:ScratchModule_07(New Actor Component Interface=DataInterface() {propertyValues {
      LocalPlayerIndex=0
      SourceActorRef=null
      SourceMode=LocalPlayer
      bRequireCurrentFrameData=true
      ActorOrComponentParameter {parameter {name=None, typeDefHandle {registeredTypeIndex=1}, varData []}}
     }}) {module=ScratchModule_07}
   &AddVelocity(
    Coordinate Space=::World
    Velocity=&Add_Vector(
     A=&Multiply_VectorByFloat(Vector=&RandomRangeVector(Minimum=<0 0 -100>, Maximum=<0 0 -300>), Float=$User.Rain_FallingSpeed)
     B=&Multiply_Vector(A=&Lerp_Vector(A=<0 0 0>, B=$Output.ScratchModule_07.Velocity, Alpha=0.15), B=$System.Rain_PlayerSpeed_Directions)
    )
   )
   /SetVariables() {isSetParameters=true, module=SetVariables_#4}
   &SphereLocation(
    Hemisphere Z=true
    MaskSpawnGroup=false
    Non Uniform Scale=<1 1 0.5>
    Offset=<0 0 350>
    Override Seed=false
    Sphere Radius=80
    Surface Only=false
    UseNonUniformScale=true
    UseOffset=true
    UseSphereOrientation=false
   ) {module=SphereLocation001}
   &ShapeLocation(
    Cylinder Radius=800
    MaskSpawnGroup=false
    Override Seed=false
    Shape Primitive=::NewEnumerator1
    Surface Only=true
    Surface Only Band Thickness=1
    Use Endcaps In Surface Only Mode=false
    UseOwnerScale=true
   )
   &self:RainTextureMasking(
    AlphaThreshold=0.5
    BoxCenter:NiagaraPosition=<-55317.5 15954.5 6841>
    BoxExtents:Vector3f=<18000 18000 1>
    Position:NiagaraPosition=$Particles.Position
    RejectionTexture=DataInterface() {propertyValues {
      Texture=&T_CrimeTown_RainMask_Buildings_01
      TextureUserParameter {parameter {name=None, typeDefHandle {registeredTypeIndex=2}, varData []}}
     }}
   ) {module=RainTextureMasking}
   /Game/Effects/Niagara/Modules/Spawn/FN_Spawn_GBuffer_Color() {enabled=false}
   &KillParticlesInVolume(
    Box Size:Vector3f=<3000 3000 400>
    Kill Shape=::NewEnumerator1
    Volume Origin:Vector3f=&AddVectorToPosition(Position=<-50410.4922 20481.1484 4000>, Vector=<0 0 0>)
   )
  ]
  particleUpdateScript [
   &ParticleState()
   &RegionMask(Region Radius=500, UseOffset=false, UseRotation=false)
   &WindForce(
    Frequency=5
    Scale=0.1
    Turbulence Flow Loop Duration=1
    Wind Speed=&Lerp_Vector(A=<50 0 0>, B=<150 0 0>, Alpha=$System.StormIntensity)
   ) {enabled=false}
   &AerodynamicDrag(Aerodynamic Drag=0.01, Use Linear Drag=true, Use Rotational Drag=true) {enabled=false}
   &SolveForcesAndVelocity(Clamp Velocity=false, Limit Acceleration=false)
   /SetVariables(Particles.PositionCenter:NiagaraPosition=&AddVectorToPosition(Position=$Engine.Owner.Position, Vector=$Emitter.OriginOffset)) {isSetParameters=true}
   &self:ScratchModule_04(
    GridSize:NiagaraFloat=&Multiply_Float(A=$Emitter.VolumeRadius, B=8)
    InputPosition:NiagaraPosition=$Particles.PositionCenter
   ) {module=ScratchModule_04}
   &ScaleSpriteSizeBySpeed(
    Max Scale Factor=<2 15>
    Min Scale Factor=<0.5 4>
    Source Velocity:Vector3f=&Multiply_Vector(A=<0.05 0.05 1>, B=$Particles.Velocity)
   )
   &AlignSpriteToMeshOrientation(SetAlignment=true, SetFacing=true)
   &ScaleColor(
    Scale Alpha=&Multiply_Float(
     A=&Multiply_Float(
      A=&FloatFromCurve(
       FloatCurve=DataInterface() {propertyValues {Curve {keys [
           {interpMode=RCIM_Cubic, tangentMode=RCTM_User, time=0, value=0}
           {interpMode=RCIM_Cubic, tangentMode=RCTM_User, time=0.5, value=1}
           {tangentMode=RCTM_User, time=1, value=0}
          ]}}}
       Scale Curve=0.35
      )
      B=$System.IndoorAlpha
     )
     B=$User.Rain_AlphaMult
    )
    Scale RGB=&MakeVectorFromLinearColorRGB(Color=&MultiplyLinearColor(Linear Color=$User.Rain_NonLitColor, Scale Factor=&MakeVector4_FromFloat(Float=$User.Rain_ColorMult)))
    ScaleA=true
    ScaleRGB=true
   )
   /SetVariables(Particles.HighEPIC COLOR:LinearColor=&MakeLinearColorFromVectorAndFloat(
     "Vector (RGB)"=&Multiply_Vector(
      A=&MakeVectorFromLinearColorRGB(Color=$Particles.Initial.Color)
      B=&VectorFromFloat(Value=&Multiply_Float(A=1, B=$User.Rain_ColorMult))
     )
     "Float (Alpha)"=&Multiply_Float(
      A=&Multiply_Float(
       A=&Multiply_Float(
        A=&FloatFromCurve(FloatCurve=DataInterface() {propertyValues {Curve {keys [
             {interpMode=RCIM_Cubic, tangentMode=RCTM_User, time=0, value=0}
             {interpMode=RCIM_Cubic, tangentMode=RCTM_User, time=0.5, value=1}
             {tangentMode=RCTM_User, time=1, value=0}
            ]}}})
        B=$System.IndoorAlpha
       )
       B=1
      )
      B=$User.Rain_AlphaMult
     )
    )) {isSetParameters=true}
   /SetVariables(Particles.Low COLOR:LinearColor=&MakeLinearColorFromVectorAndFloat(
     "Vector (RGB)"=&MakeVectorFromLinearColorRGB(Color=&MultiplyLinearColor(
       Linear Color=$User.Rain_NonLitColor_Low
       Scale Factor=&MakeVector4_FromFloat(Float=$User.Rain_ColorMult)
      ))
     "Float (Alpha)"=&Multiply_Float(
      A=&Multiply_Float(
       A=&Multiply_Float(
        A=&FloatFromCurve(FloatCurve=DataInterface() {propertyValues {Curve {keys [
             {interpMode=RCIM_Cubic, tangentMode=RCTM_User, time=0, value=0}
             {interpMode=RCIM_Cubic, tangentMode=RCTM_User, time=0.5, value=1}
             {tangentMode=RCTM_User, time=1, value=0}
            ]}}})
        B=$System.IndoorAlpha
       )
       B=1
      )
      B=$User.Rain_AlphaMult
     )
    )) {isSetParameters=true}
   &self:ScratchModule_06(Center:Vector3f=<-55317.5 15954.5 500>, Color:LinearColor=<rgba 1 0.05 0.2522 1>, Execute=true, Extents:Vector3f=<9000 9000 9000>) {enabled=false, module=ScratchModule_06}
   &DynamicMaterialParameters(
    Index 0 Param 1=1
    Module.Index0Float0Write=true
    Module.Index0Float1Write=true
    Module.Index0Float2Write=true
    Module.Index0Float3Write=true
   )
   /SetVariables(Particles.Rain_Low_Scale:Vector2f=&Multiply_Vector2D(A=$Particles.SpriteSize, B=$System.Rain_Low_Scale)) {isSetParameters=true}
  ]
  rendererClasses [&NiagaraSpriteRendererProperties]
  renderers [
   {rendererClass=&NiagaraSpriteRendererProperties}
   {rendererClass=&NiagaraSpriteRendererProperties, rendererIndex=1}
   {rendererClass=&NiagaraSpriteRendererProperties, rendererIndex=2}
  ]
 }
 SplashesClose {
  simTarget=GPUComputeSim
  emitterUpdateScript [
   &EmitterState(
    Enable Distance Culling=true
    MaxDistance=50000
    Module.bApplySpawnCountScaleByDistance=true
    Module.bUseMaxDistance=true
    Module.bUseMinDistance=false
    Scalability Mode=::NewEnumerator1
    Scale Spawn Count=true
    Spawn Count Scale By Distance=DataInterface() {propertyValues {Curve {keys [<at 0 1>,<at 0.25 1>,<at 1 0>]}}}
   )
   &SpawnRate(
    SpawnRate=&Multiply_Float(
     A=&FloatFromCurve(
      FloatCurve=DataInterface() {propertyValues {Curve {keys [<at 0.08 1>,<at 0.185 0>,<at 0.9375 0>,<at 1 1>]}}}
      CurveIndex=Linked() {linkedVariable {name=NPC.FortniteNPC.TimeOfDay_0-1, type=NiagaraFloat}}
      Scale Curve=$User.Rain_SpawnRateScale
     )
     B=&Multiply_Float(A=&Lerp_Float(A=1, B=0, Alpha=$NPC.FortniteNPC.IsPlayerIndoors), B=5000)
    )
    Use Spawn Probability=false
   )
  ]
  particleSpawnScript [
   &InitializeParticle(
    Lifetime=&RandomRangeFloat(Minimum=0.25, Maximum=0.4)
    Sprite Rotation Mode=::NewEnumerator2
    Sprite Size Mode=::NewEnumerator1
    Uniform Sprite Size Max=1
    Uniform Sprite Size Min=1
    UsePositionOffset=false
    Write Material Random=true
   )
   &AddVelocity(Velocity=&RandomRangeVector(Minimum=<0 0 25>, Maximum=<0 0 50>))
   &ShapeLocation(
    Cylinder Height=50
    Cylinder Radius=2000
    MaskSpawnGroup=false
    Offset=<0 0 -50>
    Offset Mode=::NewEnumerator0
    Override Seed=false
    Shape Primitive=::NewEnumerator1
    Surface Only=true
    Surface Only Band Thickness=0.9
    Use Endcaps In Surface Only Mode=false
    UseOwnerScale=true
   )
   /SetVariables(Particles.InitialPos:NiagaraPosition=$Particles.Position) {isSetParameters=true, module=SetVariables_#5}
   &PlaceParticlesOnDepthBuffer_GPU(Max Z Depth=1, Module.Use Max Z Depth=true)
   &MoveToNearestDistanceFieldSurface_GPU(Kill by Distance=50, bKillByDistance=true)
   &KillParticles(Kill Particles=&Boolean_NotOperation(Bool=$Output.PlaceParticlesOnDepthBuffer_GPU.ValidSample)) {module=KillParticles001}
   /SetVariables(Particles.SpriteSize=&Vector2DFromFloat(Value=&RandomRangeFloat(
      Minimum=25
      Maximum=&UniformAOrBFloat(A=35, B=&RandomRangeFloat(Minimum=40, Maximum=50), Distribution Weight=0.1)
     ))) {isSetParameters=true, module=SetVariables_#6}
   &CameraOffset(Camera Offset Amount=-25)
   &SubUVAnimation(End Frame=3, SubUV Animation Mode:ENiagara_SubUVLookupMode=::NewEnumerator2)
   &self:ScratchModule_08(Angle=10, BlendDirection=5, Normal:Vector3f=<0 0 1>, PickCeilingOrFloor=true) {module=ScratchModule_08}
   &self:RainTextureMasking(
    AlphaThreshold=0.5
    BoxCenter:NiagaraPosition=<-55317.5 15954.5 6841>
    BoxExtents:Vector3f=<18000 18000 1>
    Position:NiagaraPosition=$Particles.Position
    RejectionTexture=DataInterface() {propertyValues {
      Texture=&T_CrimeTown_RainMask_Buildings_01
      TextureUserParameter {parameter {name=None, typeDefHandle {registeredTypeIndex=2}, varData []}}
     }}
   ) {module=RainTextureMasking}
   &KillParticlesInVolume(
    Box Size:Vector3f=<3000 3000 400>
    Kill Shape=::NewEnumerator1
    Volume Origin:Vector3f=&AddVectorToPosition(Position=<-50410.4922 20481.1484 4000>, Vector=<0 0 0>)
   )
  ]
  particleUpdateScript [
   &ParticleState()
   &RegionMask(
    Inner Float=0
    Outer Float=2
    Region Falloff=0.1
    Region Radius:NiagaraFloat=$User.Rain_MaskInside_Radius
    UseOffset=false
    UseRotation=false
   ) {module=RegionMask001}
   &ScaleSpriteSize(Uniform Curve Sprite Scale=DataInterface() {propertyValues {Curve {keys [
        {arriveTangent=3.1525, interpMode=RCIM_Cubic, leaveTangent=3.1525, tangentMode=RCTM_User, time=0, value=0}
        {arriveTangent=-0.0405, interpMode=RCIM_Cubic, leaveTangent=-0.0405, tangentMode=RCTM_User, time=1, value=1}
       ]}}})
   &GravityForce(Gravity=<0 0 -150>)
   &DynamicMaterialParameters(
    Index 0 Param 1=&FloatFromCurve(
     FloatCurve=DataInterface() {propertyValues {Curve {keys [
         {arriveTangent=9.0725, interpMode=RCIM_Cubic, leaveTangent=9.0725, tangentMode=RCTM_User, time=0, value=0}
         {arriveTangent=-0.0783, interpMode=RCIM_Cubic, leaveTangent=-0.0783, tangentMode=RCTM_User, time=0.1408, value=0.9847}
         {arriveTangent=-0.0405, interpMode=RCIM_Cubic, leaveTangent=-0.0405, tangentMode=RCTM_User, time=1, value=0.5}
        ]}}}
     Scale Curve=2
    )
    Index 0 Param 2=0.1
    Index 0 Param 3=&RandomRangeFloat(Minimum=0, Maximum=0.5)
    Index 0 Param 4=1
    Module.Param0WriteEnabled=true
    Module.Param1WriteEnabled=false
    Module.Param2WriteEnabled=false
    Module.Param3WriteEnabled=false
   )
   &SolveForcesAndVelocity(Clamp Velocity=true, Limit Acceleration=false, Speed Limit=550)
   &ScaleColor(
    Scale Alpha=&Multiply_Float(
     A=&Multiply_Float(
      A=&Multiply_Float(
       A=&FloatFromCurve(
        FloatCurve=DataInterface() {propertyValues {Curve {keys [
            {interpMode=RCIM_Cubic, tangentMode=RCTM_User, time=0, value=0}
            {interpMode=RCIM_Cubic, tangentMode=RCTM_User, time=0.5, value=1}
            {tangentMode=RCTM_User, time=1, value=0}
           ]}}}
        Scale Curve=0
       )
       B=$System.IndoorAlpha
      )
      B=$Particles.RegionMask001.Float
     )
     B=$User.Rain_AlphaMult
    )
    Scale RGB=&MakeVectorFromLinearColorRGB(Color=&MultiplyLinearColor(Linear Color=$User.Rain_NonLitColor, Scale Factor=&MakeVector4_FromFloat(Float=$User.Rain_ColorMult)))
    ScaleA=true
    ScaleRGB=true
   )
   /SetVariables(Particles.HighEPIC COLOR:LinearColor=&MakeLinearColorFromVectorAndFloat(
     "Vector (RGB)"=&Multiply_Vector(
      A=&MakeVectorFromLinearColorRGB(Color=$Particles.Initial.Color)
      B=&VectorFromFloat(Value=&Multiply_Float(A=10, B=$User.Rain_ColorMult))
     )
     "Float (Alpha)"=&Multiply_Float(
      A=&Multiply_Float(
       A=&Multiply_Float(
        A=&FloatFromCurve(
         FloatCurve=DataInterface() {propertyValues {Curve {keys [
             {interpMode=RCIM_Cubic, tangentMode=RCTM_User, time=0, value=0}
             {interpMode=RCIM_Cubic, tangentMode=RCTM_User, time=0.5, value=1}
             {tangentMode=RCTM_User, time=1, value=0}
            ]}}}
         Scale Curve=2
        )
        B=$System.IndoorAlpha
       )
       B=$Particles.RegionMask001.Float
      )
      B=&Multiply_Float(A=$User.Rain_AlphaMult, B=2.5)
     )
    )) {isSetParameters=true, module=SetVariables_#7}
  ]
  rendererClasses [&NiagaraSpriteRendererProperties]
  renderers [
   {rendererClass=&NiagaraSpriteRendererProperties}
   {rendererClass=&NiagaraSpriteRendererProperties, rendererIndex=1}
  ]
 }
 SplashesFar {
  simTarget=GPUComputeSim
  emitterUpdateScript [
   &EmitterState(
    Enable Distance Culling=true
    MaxDistance=50000
    Module.bApplySpawnCountScaleByDistance=true
    Module.bUseMaxDistance=true
    Module.bUseMinDistance=false
    Scalability Mode=::NewEnumerator1
    Scale Spawn Count=true
    Spawn Count Scale By Distance=DataInterface() {propertyValues {Curve {keys [<at 0 1>,<at 0.25 1>,<at 1 0>]}}}
   )
   &SpawnRate(
    SpawnRate=&Multiply_Float(
     A=&FloatFromCurve(
      FloatCurve=DataInterface() {propertyValues {Curve {keys [<at 0.08 1>,<at 0.185 0>,<at 0.9375 0>,<at 1 1>]}}}
      CurveIndex=Linked() {linkedVariable {name=NPC.FortniteNPC.TimeOfDay_0-1, type=NiagaraFloat}}
      Scale Curve=$User.Rain_SpawnRateScale
     )
     B=&Multiply_Float(A=&Lerp_Float(A=1, B=1, Alpha=$NPC.FortniteNPC.IsPlayerIndoors), B=5000)
    )
    Use Spawn Probability=false
   )
  ]
  particleSpawnScript [
   &InitializeParticle(
    Lifetime=&RandomRangeFloat(Minimum=0.35, Maximum=1)
    Sprite Rotation Mode=::NewEnumerator2
    Sprite Size Mode=::NewEnumerator1
    Uniform Sprite Size Max=50
    Uniform Sprite Size Min=25
    UsePositionOffset=false
    Write Material Random=true
   )
   &AddVelocity(Velocity=&RandomRangeVector(Minimum=<0 0 25>, Maximum=<0 0 50>))
   &ShapeLocation(
    Cylinder Height=150
    Cylinder Radius=3500
    MaskSpawnGroup=false
    Offset=<0 0 -50>
    Offset Mode=::NewEnumerator0
    Override Seed=false
    Shape Primitive=::NewEnumerator1
    Surface Only=true
    Surface Only Band Thickness=0.5
    Use Endcaps In Surface Only Mode=false
    UseOwnerScale=true
   )
   /SetVariables(Particles.InitialPos:NiagaraPosition=$Particles.Position) {isSetParameters=true, module=SetVariables_#5}
   &PlaceParticlesOnDepthBuffer_GPU(Max Z Depth=1, Module.Use Max Z Depth=true)
   &MoveToNearestDistanceFieldSurface_GPU(Kill by Distance=50, bKillByDistance=true)
   &KillParticles(Kill Particles=&SetBoolByFloatComparison(
     A=&VectorLength(Vector=&Subtract_Vector(A=$Particles.InitialPos, B=Linked() {linkedVariable {name=Particles.Position, type=NiagaraPosition}}))
     B=1000
    )) {enabled=false}
   &KillParticles(Kill Particles=&Boolean_NotOperation(Bool=$Output.PlaceParticlesOnDepthBuffer_GPU.ValidSample)) {module=KillParticles001}
   /SetVariables(Particles.SpriteSize=&Vector2DFromFloat(Value=&RandomRangeFloat(
      Minimum=25
      Maximum=&UniformAOrBFloat(A=35, B=&RandomRangeFloat(Minimum=40, Maximum=50), Distribution Weight=0.1)
     ))) {isSetParameters=true, module=SetVariables_#6}
   &CameraOffset(Camera Offset Amount=-25)
   &SubUVAnimation(End Frame=3, SubUV Animation Mode:ENiagara_SubUVLookupMode=::NewEnumerator2)
   &self:ScratchModule_08(Angle=10, BlendDirection=5, Normal:Vector3f=<0 0 1>, PickCeilingOrFloor=true) {module=ScratchModule_08}
   &self:RainTextureMasking(
    AlphaThreshold=0.5
    BoxCenter:NiagaraPosition=<-55317.5 15954.5 6841>
    BoxExtents:Vector3f=<18000 18000 1>
    Position:NiagaraPosition=$Particles.Position
    RejectionTexture=DataInterface() {propertyValues {
      Texture=&T_CrimeTown_RainMask_Buildings_01
      TextureUserParameter {parameter {name=None, typeDefHandle {registeredTypeIndex=2}, varData []}}
     }}
   ) {module=RainTextureMasking}
  ]
  particleUpdateScript [
   &ParticleState()
   &RegionMask(
    Inner Float=0
    Outer Float=2
    Region Falloff=0.1
    Region Radius:NiagaraFloat=$User.Rain_MaskInside_Radius
    UseOffset=false
    UseRotation=false
   ) {module=RegionMask001}
   &ScaleSpriteSize(
    Uniform Curve Scale=2
    Uniform Curve Sprite Scale=DataInterface() {propertyValues {Curve {keys [
        {arriveTangent=3.1525, interpMode=RCIM_Cubic, leaveTangent=3.1525, tangentMode=RCTM_User, time=0, value=0}
        {arriveTangent=-0.0405, interpMode=RCIM_Cubic, leaveTangent=-0.0405, tangentMode=RCTM_User, time=1, value=1}
       ]}}}
   )
   &GravityForce(Gravity=<0 0 -500>)
   &DynamicMaterialParameters(
    Index 0 Param 1=&FloatFromCurve(
     FloatCurve=DataInterface() {propertyValues {Curve {keys [
         {arriveTangent=9.0725, interpMode=RCIM_Cubic, leaveTangent=9.0725, tangentMode=RCTM_User, time=0, value=0}
         {arriveTangent=-2.5249, interpMode=RCIM_Cubic, leaveTangent=-2.5249, tangentMode=RCTM_User, time=0.1408, value=0.9847}
         {arriveTangent=-0.5682, interpMode=RCIM_Cubic, leaveTangent=-0.5682, tangentMode=RCTM_User, time=0.6, value=0.3}
         {arriveTangent=-0.0405, interpMode=RCIM_Cubic, leaveTangent=-0.0405, tangentMode=RCTM_User, time=1, value=0}
        ]}}}
     Scale Curve=2
    )
    Index 0 Param 2=0.1
    Index 0 Param 3=&RandomRangeFloat(Minimum=0, Maximum=0.5)
    Index 0 Param 4=1
    Module.Param0WriteEnabled=true
    Module.Param1WriteEnabled=false
    Module.Param2WriteEnabled=false
    Module.Param3WriteEnabled=false
   )
   &SolveForcesAndVelocity(Clamp Velocity=true, Limit Acceleration=false, Speed Limit=550)
   &ScaleColor(
    Scale Alpha=&Multiply_Float(
     A=&Multiply_Float(
      A=&Multiply_Float(
       A=&FloatFromCurve(
        FloatCurve=DataInterface() {propertyValues {Curve {keys [
            {interpMode=RCIM_Cubic, tangentMode=RCTM_User, time=0, value=0}
            {interpMode=RCIM_Cubic, tangentMode=RCTM_User, time=0.5, value=1}
            {tangentMode=RCTM_User, time=1, value=0}
           ]}}}
        Scale Curve=0.02
       )
       B=$System.IndoorAlpha
      )
      B=$User.Rain_AlphaMult
     )
     B=$Particles.RegionMask001.Float
    )
    Scale RGB=&MakeVectorFromLinearColorRGB(Color=&MultiplyLinearColor(Linear Color=$User.Rain_NonLitColor, Scale Factor=&MakeVector4_FromFloat(Float=$User.Rain_ColorMult)))
    ScaleA=true
    ScaleRGB=true
   )
   /SetVariables(Particles.HighEPIC COLOR:LinearColor=&MakeLinearColorFromVectorAndFloat(
     "Vector (RGB)"=&Multiply_Vector(
      A=&MakeVectorFromLinearColorRGB(Color=$Particles.Initial.Color)
      B=&VectorFromFloat(Value=&Multiply_Float(A=8, B=$User.Rain_ColorMult))
     )
     "Float (Alpha)"=&Multiply_Float(
      A=&Multiply_Float(
       A=&Multiply_Float(
        A=&FloatFromCurve(
         FloatCurve=DataInterface() {propertyValues {Curve {keys [
             {interpMode=RCIM_Cubic, tangentMode=RCTM_User, time=0, value=0}
             {interpMode=RCIM_Cubic, tangentMode=RCTM_User, time=0.5, value=1}
             {tangentMode=RCTM_User, time=1, value=0}
            ]}}}
         Scale Curve=0.2
        )
        B=$System.IndoorAlpha
       )
       B=$Particles.RegionMask001.Float
      )
      B=$User.Rain_AlphaMult
     )
    )) {isSetParameters=true, module=SetVariables_#7}
  ]
  rendererClasses [&NiagaraSpriteRendererProperties]
  renderers [
   {rendererClass=&NiagaraSpriteRendererProperties}
   {rendererClass=&NiagaraSpriteRendererProperties, rendererIndex=1}
  ]
 }
}
summary {
 systemName=NS_CrimeTown_Rain_Camera
 emitters [
  {emitterName=Rain_Close, simTarget=GPUComputeSim, rendererClasses [&NiagaraSpriteRendererProperties]}
  {emitterName=Rain_Far, simTarget=GPUComputeSim, rendererClasses [&NiagaraSpriteRendererProperties]}
  {emitterName=RainFromObjects, simTarget=GPUComputeSim, rendererClasses [&NiagaraSpriteRendererProperties]}
  {emitterName=Rain_Mid, simTarget=GPUComputeSim, rendererClasses [&NiagaraSpriteRendererProperties]}
  {emitterName=Rain_VeryClose, simTarget=GPUComputeSim, rendererClasses [&NiagaraSpriteRendererProperties]}
  {emitterName=SplashesClose, simTarget=GPUComputeSim, rendererClasses [&NiagaraSpriteRendererProperties]}
  {emitterName=SplashesFar, simTarget=GPUComputeSim, rendererClasses [&NiagaraSpriteRendererProperties]}
  {emitterName=Rain_Mobile, rendererClasses [&NiagaraMeshRendererProperties]}
 ]
 userVariables [
  User.Rain_AlphaMult {type=NiagaraFloat, defaultValue=0.04, description=""}
  User.Rain_ColorMult {type=NiagaraFloat, defaultValue=2.5, description=""}
  User.Rain_FallingSpeed {type=NiagaraFloat, defaultValue=1.5, description=""}
  User.Rain_InheritVel {type=Vector3f, description="", defaultValue {struct=Vector3f, value=<-1 -1 0>}}
  User.Rain_MaskInside_Radius {type=NiagaraFloat, defaultValue=800, description=""}
  User.Rain_NonLitColor {type=LinearColor, description="", defaultValue {struct=LinearColor, value=<rgba 0.4586 0.5584 0.7656 1>}}
  User.Rain_NonLitColor_Low {type=LinearColor, description="", defaultValue {struct=LinearColor, value=<rgba 0.1497 0.1823 0.25 1>}}
  User.Rain_SpawnRateScale {type=NiagaraFloat, defaultValue=2.5, description=""}
 ]
}
