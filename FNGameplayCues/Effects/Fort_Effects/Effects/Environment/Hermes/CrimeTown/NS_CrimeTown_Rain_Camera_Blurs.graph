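# Niagara system for CrimeTown rain camera blurs, a single GPU emitter that continuously spawns flickering sprites in a ring or disc.
# Spawning samples GBuffer color and applies a radial region mask that feeds particle alpha.
# The vertical component of a camera vector drives curves on both sprite size and alpha, hiding the effect past a threshold.
# The building rain-mask texture rejection module is present but disabled.
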NiagaraSystem /FNGameplayCues/Effects/Fort_Effects/Effects/Environment/Hermes/CrimeTown/NS_CrimeTown_Rain_Camera_Blurs
dependencies {
 usedDataInterfaces [
  &NiagaraDataInterfaceCamera
  &NiagaraDataInterfaceGBuffer
  /Script/Niagara.NiagaraDataInterfaceTexture
  &NiagaraDataInterfaceCurve
 ]
 usedDynamicInputs []
 usedModules [&SystemState]
}
emitters {HangingParticulates {
  simTarget=GPUComputeSim
  emitterUpdateScript [
   &EmitterState()
   &SpawnRate(SpawnRate=50, Use Spawn Probability=false)
   /SetVariables() {enabled=false, isSetParameters=true}
   /SetVariables(Emitter.CamUpem:Vector3f=&Multiply_Vector(
     A=&ReturnCameraVector(Camera Vectors=::NewEnumerator2, Coordinate Space=::World, CameraQuery=DataInterface() {propertyValues {bRequireCurrentFrameData=false}})
     B=<0 0 1>
    )) {isSetParameters=true}
  ]
  particleSpawnScript [
   &InitializeParticle(
    Color=<rgba 0.5 0.5 0.5 1>
    Lifetime Mode=::NewEnumerator1
    Sprite Rotation Mode=::NewEnumerator2
    Sprite Size Mode=::NewEnumerator1
    Sprite UV Mode=::NewEnumerator3
    Uniform Sprite Size Max=15
    UsePositionOffset=false
    Write Material Random=true
   )
   &ShapeLocation(
    Disc Coverage=0.7857
    MaskSpawnGroup=false
    Offset=<0 0 100>
    Offset Mode=::NewEnumerator0
    Override Seed=false
    Ring Radius=150
    Rotation Mode=::NewEnumerator1
    Shape Primitive=::NewEnumerator4
    UseOwnerScale=true
   ) {module=ShapeLocation001}
   &AddVelocity(Velocity=&RandomRangeVector(Minimum=<0 0 0>, Maximum=<0 0 -1>))
   &ApplyInitialForces()
   /Game/Effects/Niagara/Modules/Spawn/FN_Spawn_GBuffer_Color()
   &RegionMask(Inner Float=0, Outer Float=1, Region Radius=50, UseOffset=false, UseRotation=false)
   &self:RainTextureMasking_01(
    AlphaThreshold=0.5
    BoxCenter:NiagaraPosition=<-55317.5 15954.5 6841>
    BoxExtents:Vector3f=<18000 18000 1>
    Position:NiagaraPosition=$Particles.Position
    RejectionTexture=DataInterface() {propertyValues {
      Texture=/FNGameplayCues/Effects/Fort_Effects/Effects/Environment/Hermes/CrimeTown/Textures/T_CrimeTown_RainMask_Buildings_01
      TextureUserParameter {parameter {name=None, typeDefHandle {registeredTypeIndex=2}, varData []}}
     }}
   ) {enabled=false, module=RainTextureMasking_01}
  ]
  particleUpdateScript [
   &ParticleState()
   &ScaleSpriteSize(
    Uniform Curve Scale=&FloatFromCurve(
     FloatCurve=DataInterface() {propertyValues {Curve {keys [<at 0 1>,<at 0.2 1>,<at 0.3 -1>,<at 1 0>]}}}
     CurveIndex=&Lerp_Float(A=0, B=1, Alpha=&MakeFloatFromVector(Vector=$Emitter.CamUpem, Channel=::NewEnumerator2))
    )
    Uniform Curve Sprite Scale=DataInterface() {propertyValues {Curve {keys [
        {arriveTangent=9.0725, time=0, value=1}
        {interpMode=RCIM_Cubic, tangentMode=RCTM_User, time=0.95, value=1}
        {arriveTangent=-0.0405, interpMode=RCIM_Cubic, leaveTangent=-0.0405, tangentMode=RCTM_User, time=1, value=0}
       ]}}}
   )
   &ScaleColor(
    Scale Alpha=&Multiply_Float(
     A=&Multiply_Float(
      A=&Multiply_Float(
       A=&FloatFromCurve(FloatCurve=DataInterface() {propertyValues {Curve {keys [
            {arriveTangent=2.4871, time=0, value=0.5}
            <at 0.25 1>
            {interpMode=RCIM_Cubic, tangentMode=RCTM_User, time=0.8, value=0.95}
            {arriveTangent=-0.0405, interpMode=RCIM_Cubic, leaveTangent=-0.0405, tangentMode=RCTM_User, time=1, value=0}
           ]}}})
       B=&Waveform(
        Global Phase=$Particles.Age
        "Waveform [1]"=::NewEnumerator8
        "[1] Frequency"=&RandomRangeFloat(Minimum=50, Maximum=150)
        "[1] Amplitude Min/Max"=<0.8 1>
       )
      )
      B=$Particles.RegionMask.Float
     )
     B=&FloatFromCurve(
      FloatCurve=DataInterface() {propertyValues {Curve {keys [<at 0 1>,<at 0.2 1>,<at 0.3 0>,<at 1 0>]}}}
      CurveIndex=&Lerp_Float(A=0, B=1, Alpha=&MakeFloatFromVector(Vector=$Emitter.CamUpem, Channel=::NewEnumerator2))
     )
    )
    Scale RGB=&VectorFromFloat(Value=&FloatFromCurve(
      FloatCurve=DataInterface() {propertyValues {Curve {keys [{tangentMode=RCTM_User, time=0, value=1}, {tangentMode=RCTM_User, time=1, value=0.25}]}}}
      Scale Curve=0.25
     ))
    ScaleA=true
    ScaleRGB=true
   )
   &Drag(Drag=0, Use Linear Drag=true, Use Rotational Drag=true)
   &InheritSourceMovement(ApplyForce=false, ApplyPosition=true, ApplyPositionFalloff=false) {enabled=false}
   &InheritSourceMovement(ApplyForce=false, ApplyPosition=true, ApplyPositionFalloff=false) {enabled=false, module=InheritSourceMovement001}
   &SolveForcesAndVelocity(Clamp Velocity=false, Limit Acceleration=false)
   &DynamicMaterialParameters(
    Index 0 Param 1=&RandomRangeFloat(Minimum=0.15, Maximum=0.25)
    Index 0 Param 2=0.95
    Index 0 Param 3=1
    Module.Param0WriteEnabled=true
    Module.Param1WriteEnabled=true
    Module.Param2WriteEnabled=true
    Module.Param3WriteEnabled=true
   )
  ]
  rendererClasses [&NiagaraSpriteRendererProperties]
  renderers [
   {rendererClass=&NiagaraSpriteRendererProperties}
   {rendererClass=&NiagaraSpriteRendererProperties, rendererIndex=1}
  ]
 }}
summary {
 systemName=NS_CrimeTown_Rain_Camera_Blurs
 emitters [{emitterName=HangingParticulates, simTarget=GPUComputeSim, rendererClasses [&NiagaraSpriteRendererProperties]}]
 userVariables []
}
